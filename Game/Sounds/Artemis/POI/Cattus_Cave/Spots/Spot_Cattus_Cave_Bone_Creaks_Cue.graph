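# Sound cue for one-shot bone creak spot ambience in the Cattus Cave POI.
# A Random node picks among six creak waves with equal weights and feeds a root Modulator for pitch and volume variation.
# It uses a shared closer-range tile ambient one-shot attenuation asset, an interior-space ducking ambient sound class, and a boosted volume multiplier.
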SoundCue /Game/Sounds/Artemis/POI/Cattus_Cave/Spots/Spot_Cattus_Cave_Bone_Creaks_Cue
root=Modulator_0
nodes [
 {id=Modulator_0, children [Random_1]}
 {
  id=Random_1
  children [WavePlayer_7, WavePlayer_8, WavePlayer_9, WavePlayer_10, WavePlayer_11, WavePlayer_12]
  properties {weights [1,1,1,1,1,1]}
 }
 {id=WavePlayer_10, properties {soundWaveAssetPtr=&/Spot_Cattus_Cave_Bone_Creak_02}}
 {id=WavePlayer_11, properties {soundWaveAssetPtr=&/Spot_Cattus_Cave_Bone_Creak_03}}
 {id=WavePlayer_12, properties {soundWaveAssetPtr=&/Spot_Cattus_Cave_Bone_Creak_04}}
 {id=WavePlayer_7, properties {soundWaveAssetPtr=&/Spot_Cattus_Cave_Bone_Creak_01}}
 {id=WavePlayer_8, properties {soundWaveAssetPtr=&/Spot_Cattus_Cave_Bone_Creak_05}}
 {id=WavePlayer_9, properties {soundWaveAssetPtr=&/Spot_Cattus_Cave_Bone_Creak_06}}
]
properties {
 attenuationSettings=/Game/Sounds/Attenuation/Fort_TileAmbient_OneShot_Att_Closer
 maxDistance=8500
 soundClassObject=/Game/Sounds/FortSoundClassesAndModes/Fort_Ambient_InteriorSpace_Duck_TileMix_PSM
 volumeMultiplier=2
 attenuationOverrides {bAttenuate=false}
}
