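# Sound cue for a scary vocal ambient spot effect, named for the Cattus Cave POI and built from reused Starlight waves.
# A Mixer layers a volume-boosted random pick of five breathy moans with a random pick of six crypt ghost voices.
# A Modulator randomly pitches the ghost layer down, so plays vary in both content and tone.
# Playback uses distant-creature one-shot attenuation and a ducking interior-ambience sound class.
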SoundCue /Game/Sounds/Artemis/POI/Cattus_Cave/Spots/Spot_Cattus_Cave_Scary_Voical_Cue
root=Mixer_1
nodes [
 {id=Mixer_1, children [Random_1,Modulator_0], properties {inputVolume [1.75,1]}}
 {id=Modulator_0, children [Random_3], properties {pitchMax=0.8, pitchMin=0.7}}
 {id=Random_1, children [WavePlayer_0, WavePlayer_8, WavePlayer_9, WavePlayer_10, WavePlayer_11], properties {weights [1,1,1,1,1]}}
 {
  id=Random_3
  children [WavePlayer_13, WavePlayer_14, WavePlayer_15, WavePlayer_16, WavePlayer_17, WavePlayer_22]
  properties {weights [1,1,1,1,1,1]}
 }
 {id=WavePlayer_0, properties {soundWaveAssetPtr=/Game/Sounds/Fort_Starlight/S1/SoundWaves/SpotFX_Starlight_Creepy_Breathy_Moan_St_01}}
 {id=WavePlayer_10, properties {soundWaveAssetPtr=/Game/Sounds/Fort_Starlight/S1/SoundWaves/SpotFX_Starlight_Creepy_Breathy_Moan_St_04}}
 {id=WavePlayer_11, properties {soundWaveAssetPtr=/Game/Sounds/Fort_Starlight/S1/SoundWaves/SpotFX_Starlight_Creepy_Breathy_Moan_St_05}}
 {id=WavePlayer_13, properties {soundWaveAssetPtr=/Game/Sounds/Fort_Starlight/S1/SoundWaves/SpotFX_Starlight_Crypt_Ghost_Voice_01}}
 {id=WavePlayer_14, properties {soundWaveAssetPtr=/Game/Sounds/Fort_Starlight/S1/SoundWaves/SpotFX_Starlight_Crypt_Ghost_Voice_02}}
 {id=WavePlayer_15, properties {soundWaveAssetPtr=/Game/Sounds/Fort_Starlight/S1/SoundWaves/SpotFX_Starlight_Crypt_Ghost_Voice_03}}
 {id=WavePlayer_16, properties {soundWaveAssetPtr=/Game/Sounds/Fort_Starlight/S1/SoundWaves/SpotFX_Starlight_Crypt_Ghost_Voice_04}}
 {id=WavePlayer_17, properties {soundWaveAssetPtr=/Game/Sounds/Fort_Starlight/S1/SoundWaves/SpotFX_Starlight_Crypt_Ghost_Voice_05}}
 {id=WavePlayer_22, properties {soundWaveAssetPtr=/Game/Sounds/Fort_Starlight/S1/SoundWaves/SpotFX_Starlight_Crypt_Ghost_Voice_06}}
 {id=WavePlayer_8, properties {soundWaveAssetPtr=/Game/Sounds/Fort_Starlight/S1/SoundWaves/SpotFX_Starlight_Creepy_Breathy_Moan_St_02}}
 {id=WavePlayer_9, properties {soundWaveAssetPtr=/Game/Sounds/Fort_Starlight/S1/SoundWaves/SpotFX_Starlight_Creepy_Breathy_Moan_St_03}}
]
properties {
 attenuationSettings=/Game/Sounds/Attenuation/Fort_TileAmbient_OneShot_Distant_Creature_Att
 maxDistance=8500
 soundClassObject=/Game/Sounds/FortSoundClassesAndModes/Fort_Ambient_InteriorSpace_Duck_TileMix_PSM
 volumeMultiplier=1
}
